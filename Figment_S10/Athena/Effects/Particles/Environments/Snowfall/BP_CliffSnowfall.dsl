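; Cliff snowfall Blueprint that emits falling-snow particles along a spline.
; Each particle spawn picks a seeded stream value, samples the spline's local location at that time and passes it to the particle system as SplineLocation.
; The same value chooses one of two opposite sideways offsets for the SplineSwitch parameter.
; The particle system deactivates on BeginPlay during replay playback.
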(event EventBeginPlay
  (if (Replay|GetIsPlayingReplay)
    (Components|Activation|Deactivate (Variables|Default|GetP_Cattus_Cliff_Falling))))

(event OnParticleSpawn(P_Cattus_Cliff_Falling)
  (params (EventName :type Name :default "None") (EmitterTime :type Float%20%28single-precision%29) (Location :type Vector :default "0, 0, 0") (Velocity :type Vector :default "0, 0, 0"))
  (bind _random_stream (Variables|Default|GetRandom%20Stream))
  (bind _returnvalue (Math|Random|RandomFloatinRangefromStream _random_stream 0.0 1.0))
  (bind _p_cattus_cliff_falling (Variables|Default|GetP_Cattus_Cliff_Falling))
  (Effects|Components|ParticleSystem|SetVectorParameter _p_cattus_cliff_falling "SplineLocation" (Spline|GetLocationatTime (Variables|Default|GetSpline) _returnvalue "Local" false))
  (Effects|Components|ParticleSystem|SetVectorParameter _p_cattus_cliff_falling "SplineSwitch" (select (>= _returnvalue 0.5) "0.000000,-600.000000,0.000000" "0.000000,600.000000,0.000000")))

(fn ConstructionScript)
